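# Lit mesh-particle material for the Boogie Bomb's overhead disco ball.
# Particle dynamic parameters drive world position offset, spinning the mesh around Z and deforming it with radial scaling and sine-based vertical motion.
# Disco ball textures supply optionally tinted color, metallic shine and mask-driven roughness.
# Emissive masks a rotation-corrected Fresnel with particle color and a panning, reflection-mapped rainbow, which Low quality replaces with white.
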Material /Game/Weapons/Grenades/Grenade_Disco/Materials/M_BoogieBomb_Overhead
graph {
 properties {bUsedWithMeshParticles=true, bUsedWithNiagaraMeshParticles=true, pixelDepthOffsetMode=PDOM_Legacy, usedShadingModels=Default Lit}
 nodes [
  n1 = Constant3Vector() {constant=<rgba 0 0 1 0>}
  n2 = DynamicParameter() {defaultValue=<rgba 0 -0.25 1 1>, paramNames [WPO,RotateSpeed,Param3,Param4]}
  n3 = Multiply(A=@n2.RotateSpeed) {constB=-1}
  n4 = WorldPosition()
  n5 = Add(A=ObjectPositionWS(), B=Multiply(A=Constant3Vector() {constant=<rgba 0 0 -0.25 0>}, B=ScalarParameter() {defaultValue=50}) {constA=-0.25})
  n6 = Multiply(A=ComponentMask(None=Subtract(A=@n4.XYZ, B=@n5)) {g=true, r=true}, B=Power(Base=@n2.WPO) {constExponent=5})
  n7 = Power(Base=@n2.WPO)
  n8 = Clamp(None=LinearInterpolate(Alpha=@n2.WPO) {constA=1, constB=0})
  n9 = Add(A=Multiply(A=@n8, B=LinearInterpolate(Alpha=@n8) {constA=1, constB=3})) {constB=0.25}
  n10 = Multiply(A=@n7, B=Multiply(A=@n7, B=Sine(None=@n9) {period=0.75}))
  Add(
   A=RotateAboutAxis(NormalizedRotationAxis=@n1, RotationAngle=@n3, PivotPoint=ObjectPositionWS(), Position=WorldPosition().XYZ)
   B=Multiply(A=VectorParameter() {parameterName=WPOMax, defaultValue=<rgba -1 -1 -100 1>}.RGB, B=AppendVector(A=@n6, B=@n10)) {constB=0}
  )
  n11 = TextureSample() {texture=/Game/Environments/Props/Retail/Textures/T_DiscoBall_D}
  Clamp(None=LinearInterpolate(Alpha=@n11.G) {constA=0.5, constB=-0.4}, Min=ScalarParameter() {defaultValue=0.03, parameterName=MinRoughness}) {minDefault=0.02}
  Constant() {r=1}
  LinearInterpolate(
   A=@n11.RGB
   B=Multiply(A=VectorParameter() {parameterName=Color, defaultValue=<rgba 1 0.587 0 1>}.RGB, B=@n11.RGB)
   Alpha=ScalarParameter() {parameterName=Use Color Multiply}
  )
  n12 = TextureSample() {samplerType=SAMPLERTYPE_Normal, texture=/Game/Environments/Props/Retail/Textures/T_DiscoBall_N}
  n13 = Fresnel(Normal=MaterialFunctionCall(Rotation Axis=@n1, Rotation Angle=@n3, World Space Vertex Normals=Transform(None=@n12.RGB)) {materialFunction=&FixRotateAboutAxisNormals}.New World Space Vertex Normal) {baseReflectFraction=0, exponent=4}
  n14 = ParticleColor()
  n15 = Multiply(A=ComponentMask(None=ReflectionVectorWS()) {g=true, r=true}, B=Fresnel() {baseReflectFraction=0, exponent=0.5})
  n16 = LinearInterpolate(B=TextureSample(UVs=Panner(Coordinate=@n15) {speedX=1, speedY=1}) {texture=/Game/Global/Textures/Default/Masks_and_Gradients/T_Grad_Rainbow}.RGB) {constA=1, constAlpha=0.75}
  n17 = Constant() {r=1}
  n18 = LinearInterpolate(
   A=Multiply(A=@n14.RGB, B=QualitySwitch(Default=@n16, Low=@n17))
   B=@n14.RGB
   Alpha=ScalarParameter() {parameterName=Use Particle Color Flash}
  )
  MaterialFunctionCall(Emissive=Multiply(A=@n11.G, B=Multiply(A=@n13, B=@n18))) {materialFunction=&MF_FX_EmissiveAdjustments}
 ]
}
