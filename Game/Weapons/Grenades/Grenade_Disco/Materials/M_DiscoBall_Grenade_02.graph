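# Default Lit material for the Disco grenade, using material attributes and enabled for skeletal, spline and instanced meshes.
# Surface attributes come from the shared gun master shader function and pass through the weapon skin FX function.
# World Position Offset sums the LOD pickup/cull-fade effect and a black hole function, which samples the LOD-displaced position and a vector parameter.
# An FX emissive adjustment function processes the combined attributes last.
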Material /Game/Weapons/Grenades/Grenade_Disco/Materials/M_DiscoBall_Grenade_02
graph {
 properties {
  bUseMaterialAttributes=true
  bUsedWithInstancedStaticMeshes=true
  bUsedWithMeshDeformer=true
  bUsedWithMorphTargets=true
  bUsedWithSkeletalMesh=true
  bUsedWithSplineMeshes=true
  pixelDepthOffsetMode=PDOM_Legacy
  usedShadingModels=Default Lit
 }
 nodes [
  n1 = MaterialFunctionCall!(UV 1 coordinates, MuzzleFalloff) {materialFunction=/Game/Packages/Fortnite_UniqueMaterials_1/Weapons/M_GunMasterShaderFunction}
  n2 = VectorParameter() {parameterName=WorldPostionOfBlackHole, defaultValue=<rgba -30000 25536 1952 0>}
  n3 = MaterialFunctionCall!(
   World Position=WorldPosition().XYZ
   DistanceCullFadeAnimationOffset=MaterialFunctionCall() {materialFunction=/Game/Packages/Fortress_SharedMaterials/Base_Material_Functions/PickupEffectDistanceCullOverride}.Result
   Use actor center instead of pivot point
  ) {materialFunction=/Game/Packages/Fortress_SharedMaterials/Base_Material_Functions/LODEffect}
  n4 = MaterialFunctionCall!(WorldPostionOfBlackHole=AppendVector(A=@n2.RGB, B=@n2.RGB), World Position=Add(A=@n3.Result, B=WorldPosition().XYZ)) {materialFunction=/Game/Packages/Fortnite_UniqueMaterials_1/BlackHoleFunction}
  MaterialFunctionCall!(In=SetMaterialAttributes!(
    MaterialAttributes=MaterialFunctionCall!(Pre-Skinned World Normal, Gmap, Customization Mask, IsCharacter, IsStatsB, Material Attributes=@n1.Material Result) {materialFunction=/Game/Weapons/WeaponSkins/MaterialFunction/Weapon_SkinFXFunction}.Result
    World Position Offset=Add(A=@n4.Result, B=@n3.Result)
   ) {attributeSetTypes [#1]}) {materialFunction=/Game/Effects/Niagara/MasterMaterials/Functions/MF_FX_EmissiveAdjustments_MA}
 ]
}
